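# Control Rig for the Teriyaki Fish Assassin one-handed pickaxe that procedurally wiggles the fish's spine bones.
# Two chained BoneHarmonics units run on execution: one waves dyn_spine1 to 3, the other dyn_spine4 and 5 with different amplitude, speed and offset.
# Both add slight noise on Y and Z and propagate to children, carrying the head, jaw and fins along.
# The tail branches from dyn_spineBase and is not animated here.
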ControlRigBlueprint /BRCosmetics/Weapons/FORT_Melee/Pickaxe_Teriyaki_Fish_Assassin_1h/Meshes/Teriyaki_Fish_Assassin_1h_CtrlRig
graphs [
 RigVMFunctionLibrary {links [], nodes []}
 RigVMModel {nodes [
   n1 = &RigUnit_BeginExecution()
   /Script/ControlRig.RigUnit_BoneHarmonics(
    Bones=((Bone="dyn_spine1",Ratio=0),(Bone="dyn_spine2",Ratio=0),(Bone="dyn_spine3",Ratio=0))
    RotationOrder=YZX
    WaveAmplitude=(X=0,Y=3,Z=2)
    WaveEase=Linear
    WaveFrequency=(X=1,Y=4,Z=2)
    WaveMaximum=1
    WaveMinimum=0.5
    WaveNoise=(X=0,Y=0.1,Z=0.1)
    WaveOffset=(X=0,Y=0,Z=0)
    WaveSpeed=(X=1,Y=1.5,Z=1)
    bPropagateToChildren=True
    ExecutePin=@n1.ExecutePin
   ) {is=struct RigVMUnitNode}
   > &RigUnit_BoneHarmonics(
    Bones=((Bone="dyn_spine4",Ratio=0),(Bone="dyn_spine5",Ratio=0))
    RotationOrder=YZX
    WaveAmplitude=(X=0,Y=2,Z=3)
    WaveEase=Linear
    WaveFrequency=(X=1,Y=4,Z=2)
    WaveMaximum=1
    WaveMinimum=0.5
    WaveNoise=(X=0,Y=0.1,Z=0.1)
    WaveOffset=(X=0,Y=2,Z=2)
    WaveSpeed=(X=1,Y=1,Z=1.5)
    bPropagateToChildren=True
   )
  ]}
]
hierarchy [
 b_AR_Root {loc=<0 0 0>, rot=<0 0 0.7071 0.7071>}
 Melee_TwoHanded_Handle {parent=b_AR_Root, loc=<8.1477 -0.008 -29.0793>, rot=<0 0 0.7071 -0.7071>}
 dyn_spineBase {parent=Melee_TwoHanded_Handle, loc=<0.072 -7.7493 29.6555>, rot=<0 0.6289 0 -0.7775>}
 dyn_spine1 {parent=dyn_spineBase, loc=<6.1778 0 0>, rot=<0 0.099 0 -0.9951>}
 dyn_spine2 {parent=dyn_spine1, loc=<6.7918 0 0>, rot=<0 0.032 0 -0.9995>}
 dyn_spine3 {parent=dyn_spine2, loc=<4.8395 0 0>, rot=<0 0.1122 0 -0.9937>}
 dyn_spine4 {parent=dyn_spine3, loc=<6.4098 0 0>, rot=<0 0.0628 0 -0.998>}
 dyn_spine5 {parent=dyn_spine4, loc=<4.5328 0 0>, rot=<0 0.0797 0 -0.9968>}
 dyn_neck {parent=dyn_spine5, loc=<4.9696 0 0>, rot=<0 0.1715 0 -0.9852>}
 dyn_head {parent=dyn_neck, loc=<4.3351 0 0>, rot=<0 0.2181 0 -0.9759>}
 dyn_jaw {parent=dyn_head, loc=<6.8331 0 0>, rot=<0 0 0 1>}
 dyn_finTop {parent=dyn_spine5, loc=<-1.5385 0 -4.1635>, rot=<0 -0.7575 0 -0.6528>}
 dyn_finLeft1 {parent=dyn_spine5, loc=<-1.2796 6.2018 5.1681>, rot=<-0.2964 -0.9005 0.0995 0.3022>}
 dyn_finLeft2 {parent=dyn_finLeft1, loc=<4.5676 0 0>, rot=<0 0 0 -1>}
 dyn_finRight1 {parent=dyn_spine5, loc=<-1.235 -5.9254 4.9323>, rot=<0.3122 -0.8877 -0.1123 0.3193>}
 dyn_finRight2 {parent=dyn_finRight1, loc=<5.3258 0 0>, rot=<0 0 0 -1>}
 dyn_tailBase {parent=dyn_spineBase, loc=<-5.5508 0.0022 -0.674>, rot=<0.0002 -0.9957 0.0002 -0.0924>}
 dyn_tailBottom {parent=dyn_tailBase, loc=<2.8913 -0.0015 -2.3686>, rot=<0 -0.3917 0 -0.9201>}
 dyn_tailTop {parent=dyn_tailBase, loc=<3.0375 0 1.9229>, rot=<0 0.3071 0 -0.9517>}
]
