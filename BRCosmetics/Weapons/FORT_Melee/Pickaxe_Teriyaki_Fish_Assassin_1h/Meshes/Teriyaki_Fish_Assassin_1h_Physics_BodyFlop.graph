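# Physics asset for the Teriyaki Fish Assassin one-handed pickaxe, named for a body-flop motion on the fish.
# A box body on dyn_spineBase anchors a chain of five identical capsules on dyn_spine1 through dyn_spine5.
# Each constraint links a bone to the one before it with tight twist and swing limits.
# Swing range generally widens toward the last bone, so the far end bends more than the base.
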PhysicsAsset /BRCosmetics/Weapons/FORT_Melee/Pickaxe_Teriyaki_Fish_Assassin_1h/Meshes/Teriyaki_Fish_Assassin_1h_Physics_BodyFlop
bodies {
 dyn_spine1 [{shapeType=Capsule, length=4, radius=4, center=<3 0 0>, rotation=<rot 90 0 0>}]
 dyn_spine2 [{shapeType=Capsule, length=4, radius=4, center=<3 0 0>, rotation=<rot 90 0 0>}]
 dyn_spine3 [{shapeType=Capsule, length=4, radius=4, center=<3 0 0>, rotation=<rot 90 0 0>}]
 dyn_spine4 [{shapeType=Capsule, length=4, radius=4, center=<3 0 0>, rotation=<rot 90 0 0>}]
 dyn_spine5 [{shapeType=Capsule, length=4, radius=4, center=<3 0 0>, rotation=<rot 90 0 0>}]
 dyn_spineBase [{shapeType=Box, extentX=6.2137, extentY=5.6674, extentZ=7.0137, center=<3.1554 0 0.7635>}]
}
constraints [
 {bone1Name=dyn_spine1, bone2Name=dyn_spineBase, swing1LimitDegrees=10, swing2LimitDegrees=5, twistLimitDegrees=5}
 {bone1Name=dyn_spine2, bone2Name=dyn_spine1, swing1LimitDegrees=12, swing2LimitDegrees=8, twistLimitDegrees=5}
 {bone1Name=dyn_spine3, bone2Name=dyn_spine2, swing1LimitDegrees=11, swing2LimitDegrees=10, twistLimitDegrees=5}
 {bone1Name=dyn_spine4, bone2Name=dyn_spine3, swing1LimitDegrees=12, swing2LimitDegrees=13, twistLimitDegrees=5}
 {bone1Name=dyn_spine5, bone2Name=dyn_spine4, swing1LimitDegrees=15, swing2LimitDegrees=16, twistLimitDegrees=5}
]
