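# Physics asset for the fins, tail and jaw of the one-handed Teriyaki Fish Assassin pickaxe mesh.
# Box bodies on dyn_ bones are constrained in a tree rooted at dyn_spineBase, with two-segment side fins and a tail base that splits into top and bottom.
# Angular limits keep motion small, with side fins swinging most freely, the top fin mostly on one axis, and the tail parts kept stiffer.
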PhysicsAsset /BRCosmetics/Weapons/FORT_Melee/Pickaxe_Teriyaki_Fish_Assassin_1h/Meshes/Teriyaki_Fish_Assassin_1h_Physics_Fins
bodies {
 dyn_finLeft1 [{shapeType=Box, extentX=5, extentY=8, extentZ=2, center=<3 0 0>}]
 dyn_finLeft2 [{shapeType=Box, extentX=5, extentY=8, extentZ=2, center=<3 0 0>}]
 dyn_finRight1 [{shapeType=Box, extentX=4, extentY=8, extentZ=2, center=<3 0 0>}]
 dyn_finRight2 [{shapeType=Box, extentX=4, extentY=8, extentZ=2, center=<3 0 0>}]
 dyn_finTop [{shapeType=Box, extentX=5.5, extentY=1.5, extentZ=18, center=<3 0 0>}]
 dyn_jaw [{shapeType=Box, extentX=5, extentY=8, extentZ=3, center=<2.5 0 1.5>}]
 dyn_spineBase [{shapeType=Box, extentX=2, extentY=2, extentZ=2}]
 dyn_tailBase [{shapeType=Box, extentX=4, extentY=3, extentZ=4, center=<2 0 0>}]
 dyn_tailBottom [{shapeType=Box, extentX=8, extentY=3, extentZ=4, center=<4 0 0>}]
 dyn_tailTop [{shapeType=Box, extentX=8, extentY=3, extentZ=4, center=<4 0 0>}]
}
constraints [
 {bone1Name=dyn_finTop, bone2Name=dyn_spineBase, swing1LimitDegrees=30, swing2LimitDegrees=1, twistLimitDegrees=1}
 {bone1Name=dyn_finLeft1, bone2Name=dyn_spineBase, swing1LimitDegrees=30, swing2LimitDegrees=30, twistLimitDegrees=5}
 {bone1Name=dyn_finLeft2, bone2Name=dyn_finLeft1, swing1LimitDegrees=30, swing2LimitDegrees=30, twistLimitDegrees=5}
 {bone1Name=dyn_finRight1, bone2Name=dyn_spineBase, swing1LimitDegrees=30, swing2LimitDegrees=30, twistLimitDegrees=5}
 {bone1Name=dyn_finRight2, bone2Name=dyn_finRight1, swing1LimitDegrees=30, swing2LimitDegrees=30, twistLimitDegrees=5}
 {bone1Name=dyn_tailBase, bone2Name=dyn_spineBase, swing1LimitDegrees=10, swing2LimitDegrees=10, twistLimitDegrees=5}
 {bone1Name=dyn_tailBottom, bone2Name=dyn_tailBase, swing1LimitDegrees=10, swing2LimitDegrees=10, twistLimitDegrees=5}
 {bone1Name=dyn_tailTop, bone2Name=dyn_tailBase, swing1LimitDegrees=10, swing2LimitDegrees=10, twistLimitDegrees=5}
 {bone1Name=dyn_jaw, bone2Name=dyn_spineBase, swing1LimitDegrees=10, swing2LimitDegrees=30, twistLimitDegrees=1}
]
