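# Animation Blueprint for the Teriyaki Fish Assassin one-handed pickaxe cosmetic, adding secondary motion to its mesh.
# The input pose passes through two chained RigidBody simulations in base-bone space on dyn_spineBase, first a fins physics asset, then a body-flop asset with lighter gravity and mostly softer acceleration response.
# A Control Rig then post-processes the simulated pose before output.
# The event graph holds no connected logic.
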AnimBlueprint /BRCosmetics/Weapons/FORT_Melee/Pickaxe_Teriyaki_Fish_Assassin_1h/Meshes/Teriyaki_Fish_Assassin_1h_AnimBP
target_skeleton=&/Teriyaki_Fish_Assassin_1h_Skeleton
graphs [
 AnimGraph {nodes [
   n1 = RigidBody!(ComponentPose=Local To Component(LocalPose=Input Pose()), Alpha, ExternalForce, ComponentLinearVelScale:Vector="1,1,1") -> (Pose) {
    bClampLinearTranslationLimitToRefPose=true
    bOverrideWorldGravity=true
    overridePhysicsAsset=&/Teriyaki_Fish_Assassin_1h_Physics_Fins
    simulationSpace=BaseBoneSpace
    alphaBoolBlend {blendOption=Linear}
    baseBoneRef {boneName=dyn_spineBase}
    componentAppliedLinearAccClamp=<4000 4000 4000>
    componentLinearAccScale=<0.3 0.3 0.3>
    componentLinearVelScale=<1 1 1>
    overrideWorldGravity {z=-980}
   }
   n2 = RigidBody!(ComponentPose=@n1, Alpha, ExternalForce) -> (Pose) {
    bClampLinearTranslationLimitToRefPose=true
    bOverrideWorldGravity=true
    overridePhysicsAsset=&/Teriyaki_Fish_Assassin_1h_Physics_BodyFlop
    simulationSpace=BaseBoneSpace
    alphaBoolBlend {blendOption=Linear}
    baseBoneRef {boneName=dyn_spineBase}
    componentAppliedLinearAccClamp=<3000 3000 1000>
    componentLinearAccScale=<0.15 0.3 0.25>
    componentLinearVelScale=<0.5 0.5 0.4>
    overrideWorldGravity {z=-600}
   }
   n3 = Control Rig(Source=Component To Local(ComponentPose=@n2)) {
    alphaBoolBlend {blendOption=Linear}
    controlRigAssetReference {blueprintRigClass=&/Teriyaki_Fish_Assassin_1h_CtrlRig.Teriyaki_Fish_Assassin_1h_CtrlRig_C}
   }
   Output Pose(Result=@n3) {layerGroup=None}
  ]}
 EventGraph {nodes [
   TryGetPawnOwner()
   Event BlueprintUpdateAnimation()
  ]}
]
variables []
